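# Player animation Blueprint for dual-wieldable weapons, adding independent left and right hand reload animations through linked-layer graphs.
# ItemFinalPoseOverride runs one state machine per side, layering its reload sequence over the cached final pose from that clavicle with mesh-space rotation blending.
# MovementMode_AimOffset uses negative-depth branch filters to exclude reloading arms from the aim offset.
# Transitions read reload booleans not declared here; other layers contain no logic.
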AnimBlueprint /VividRazorWeaponGameplay/Assets/Animation/DualWieldable_Player_AnimBP
target_skeleton=&Fortnite_M_Avg_Player_Skeleton
graphs [
 AnimGraph {nodes [Output Pose()]}
 EventGraph {links [], nodes []}
 ItemAimOffset {nodes [
   Input Pose!() -> (Pose, AimOffsetAlpha:double, Yaw:double, Pitch:double) {name=InPose_AimOffset}
   Input Pose!() -> (Pose) {name=InPose_PreAimOffsets}
   Output Pose() {layerGroup=Item Override}
  ]}
 ItemFallAdditive {nodes [
   Input Pose() {name=InPoseItemFallAdditive}
   Output Pose() {layerGroup=Item Override}
  ]}
 ItemFallLandAdditive {nodes [
   Input Pose() {name=InPoseItemFallLandAdditive}
   Output Pose() {layerGroup=Item Override}
  ]}
 ItemFallLevel2Override {nodes [
   Input Pose() {name=InPose_FallLevel2}
   Output Pose() {layerGroup=Item Override}
  ]}
 ItemFinalPoseOverride {nodes [
   AnimGraphNode_StateMachine_3 = Right Reload SM() {is=title AnimGraphNode_StateMachine}
   n1 = Use cached pose 'Final Input Pose'()
   AnimGraphNode_StateMachine_2 = Left Reload SM() {is=title AnimGraphNode_StateMachine}
   Output Pose(Result=Layered blend per bone(
     BasePose=Layered blend per bone(BasePose=@n1, BlendPoses_0=@AnimGraphNode_StateMachine_3) {bMeshSpaceRotationBlend=true, layerSetup [{branchFilters [{boneName=clavicle_r}]}]}
     BlendPoses_0=@AnimGraphNode_StateMachine_2
    ) {bMeshSpaceRotationBlend=true, layerSetup [{branchFilters [{boneName=clavicle_l}]}]}) {layerGroup=Item Override}
   Save cached pose 'Final Input Pose'(Pose=Input Pose() {name=InFinalPose}) {cacheName=Final Input Pose}
  ]}
 Left Reload SM {
  entry=PassThrough
  path=ItemFinalPoseOverride.AnimGraphNode_StateMachine_2.Left Reload SM
  ids=id
  nodes [
   AnimStateNode_0 = Reload Left Single() {is=title AnimStateNode}
   AnimStateNode_1 = PassThrough() {is=title AnimStateNode}
   AnimStateTransitionNode_0 = AnimStateTransitionNode()
   AnimStateTransitionNode_1 = AnimStateTransitionNode() {crossfadeDuration=0.3}
   AnimStateTransitionNode_4 = AnimStateTransitionNode() {crossfadeDuration=0.3}
  ]
  links [
   {
    dst=Reload Left Single
    rule=ItemFinalPoseOverride.AnimGraphNode_StateMachine_2.Left Reload SM.AnimStateTransitionNode_1.Transition
    src=PassThrough
    via=AnimStateTransitionNode_1
   }
   {
    dst=PassThrough
    rule=ItemFinalPoseOverride.AnimGraphNode_StateMachine_2.Left Reload SM.AnimStateTransitionNode_0.Transition
    src=Reload Left Single
    via=AnimStateTransitionNode_0
   }
   {
    dst=PassThrough
    rule=ItemFinalPoseOverride.AnimGraphNode_StateMachine_2.Left Reload SM.AnimStateTransitionNode_4.Transition
    src=Reload Left Single
    via=AnimStateTransitionNode_4
   }
  ]
 }
 Reload Left Single {
  path=ItemFinalPoseOverride.AnimGraphNode_StateMachine_2.Left Reload SM.AnimStateNode_0.Reload Left Single
  nodes [Output Animation Pose(Result=Sequence Player 'DualWieldable_Reload_LeftOnly'!(PlayRate:float=1) -> (Pose) {bLoopAnimation=false, sequence=&/DualWieldable_Reload_LeftOnly})]
 }
 PassThrough {
  path=ItemFinalPoseOverride.AnimGraphNode_StateMachine_2.Left Reload SM.AnimStateNode_1.PassThrough
  nodes [Output Animation Pose(Result=Use cached pose 'Final Input Pose'())]
 }
 Transition {
  path=ItemFinalPoseOverride.AnimGraphNode_StateMachine_2.Left Reload SM.AnimStateTransitionNode_0.Transition
  nodes [Result(bCanEnterTransition=NOT Boolean(A=Get bIsSecondaryReload() -> (bIsSecondaryReload:bool)))]
 }
 Transition {
  path=ItemFinalPoseOverride.AnimGraphNode_StateMachine_2.Left Reload SM.AnimStateTransitionNode_1.Transition
  nodes [Result(bCanEnterTransition=Get bIsSecondaryReloadStart() -> (bIsSecondaryReloadStart:bool))]
 }
 Transition {
  path=ItemFinalPoseOverride.AnimGraphNode_StateMachine_2.Left Reload SM.AnimStateTransitionNode_4.Transition
  nodes [Result(bCanEnterTransition=Get bIsEarlySecondaryReloadBlendOut() -> (bIsEarlySecondaryReloadBlendOut:bool))]
 }
 Right Reload SM {
  entry=PassThrough
  path=ItemFinalPoseOverride.AnimGraphNode_StateMachine_3.Right Reload SM
  ids=id
  nodes [
   AnimStateNode_0 = Reload Right Single() {is=title AnimStateNode}
   AnimStateNode_1 = PassThrough()
   AnimStateTransitionNode_0 = AnimStateTransitionNode() {crossfadeDuration=0.3}
   AnimStateTransitionNode_1 = AnimStateTransitionNode()
   AnimStateTransitionNode_3 = AnimStateTransitionNode() {crossfadeDuration=0.3}
  ]
  links [
   {
    dst=Reload Right Single
    rule=ItemFinalPoseOverride.AnimGraphNode_StateMachine_3.Right Reload SM.AnimStateTransitionNode_0.Transition
    src=PassThrough
    via=AnimStateTransitionNode_0
   }
   {
    dst=PassThrough
    rule=ItemFinalPoseOverride.AnimGraphNode_StateMachine_3.Right Reload SM.AnimStateTransitionNode_1.Transition
    src=Reload Right Single
    via=AnimStateTransitionNode_1
   }
   {
    dst=PassThrough
    rule=ItemFinalPoseOverride.AnimGraphNode_StateMachine_3.Right Reload SM.AnimStateTransitionNode_3.Transition
    src=Reload Right Single
    via=AnimStateTransitionNode_3
   }
  ]
 }
 Reload Right Single {
  path=ItemFinalPoseOverride.AnimGraphNode_StateMachine_3.Right Reload SM.AnimStateNode_0.Reload Right Single
  nodes [Output Animation Pose(Result=Sequence Player 'DualWieldable_Reload_RightOnly'!(PlayRate:float=1) -> (Pose) {bLoopAnimation=false, sequence=&/DualWieldable_Reload_RightOnly})]
 }
 PassThrough {
  path=ItemFinalPoseOverride.AnimGraphNode_StateMachine_3.Right Reload SM.AnimStateNode_1.PassThrough
  nodes [Output Animation Pose(Result=Use cached pose 'Final Input Pose'())]
 }
 Transition {
  path=ItemFinalPoseOverride.AnimGraphNode_StateMachine_3.Right Reload SM.AnimStateTransitionNode_0.Transition
  nodes [Result(bCanEnterTransition=Get bIsPrimaryReloadStart() -> (bIsPrimaryReloadStart:bool))]
 }
 Transition {
  path=ItemFinalPoseOverride.AnimGraphNode_StateMachine_3.Right Reload SM.AnimStateTransitionNode_1.Transition
  nodes [Result(bCanEnterTransition=NOT Boolean(A=Get bIsPrimaryReload() -> (bIsPrimaryReload:bool)))]
 }
 Transition {
  path=ItemFinalPoseOverride.AnimGraphNode_StateMachine_3.Right Reload SM.AnimStateTransitionNode_3.Transition
  nodes [Result(bCanEnterTransition=Get bIsEarlyPrimaryReloadBlendOut() -> (bIsEarlyPrimaryReloadBlendOut:bool))]
 }
 ItemFullBodyAimOffsets {nodes [
   Input Pose!() -> (Pose, FullBody_AimOffsetAlpha:double, Aim_Yaw:double, Aim_Pitch:double) {name=InPoseFullBodyAimOffsets}
   Input Pose!() -> (Pose) {name=InPosePreFullBodyAimOffsets}
   Output Pose() {layerGroup=Item Override}
  ]}
 ItemFullBodyOverride {nodes [
   Input Pose() {name=InPoseFullBody}
   Output Pose() {layerGroup=Item Override}
  ]}
 ItemIdleOverride {nodes [
   Input Pose() {name=InPose_Idle}
   Output Pose() {layerGroup=Item Override}
  ]}
 ItemJetPackJumpAdditive {nodes [Output Pose() {layerGroup=Item Override}]}
 ItemJetPackStartAdditive {nodes [Output Pose() {layerGroup=Item Override}]}
 ItemJumpLoopAdditive {nodes [
   Input Pose() {name=InPoseItemJumpLoopAdditive}
   Output Pose() {layerGroup=Item Override}
  ]}
 ItemJumpOverride {nodes [
   Input Pose() {name=InPoseJump}
   Output Pose() {layerGroup=Item Override}
  ]}
 ItemJumpUpAdditive {nodes [
   Input Pose() {name=InPoseItemJumpUpAdditive}
   Output Pose() {layerGroup=Item Override}
  ]}
 ItemPreAimOffsets {nodes [
   Input Pose() {name=InPosePreAimOffsets}
   Output Pose() {layerGroup=Item Override}
  ]}
 ItemPreIK {nodes [
   Input Pose() {name=InPose_PreIK}
   Output Pose() {layerGroup=Item Override}
  ]}
 ItemRightHandIK {nodes [
   Input Pose() {name=InPosePreRightHandIK}
   Input Pose() {name=InPosePostRightHandIK}
   Output Pose() {layerGroup=Item Override}
  ]}
 ItemSkeletalControl {nodes [
   Input Pose()
   Output Pose() {layerGroup=Item Override}
  ]}
 ItemSwimJumpFallLoop {nodes [Output Pose() {layerGroup=Item Override}]}
 ItemSwimJumpSurfaceEnd {nodes [Output Pose() {layerGroup=Item Override}]}
 ItemSwimJumpSurfaceLoop {nodes [Output Pose() {layerGroup=Item Override}]}
 ItemUpperBody {nodes [
   Input Pose!() -> (Pose, InputParam:Fort Anim Input Adjusted Aim Structure) {name=InPoseUpperBody}
   Output Pose() {layerGroup=Item Override}
  ]}
 ItemZipLineStartAdditive {nodes [Output Pose() {layerGroup=Item Override}]}
 MovementMode_AimOffset {nodes [
   AnimGraphNode_StateMachine_0 = Single Reload Aim SM() {is=title AnimGraphNode_StateMachine}
   Output Pose(Result=@AnimGraphNode_StateMachine_0) {layerGroup=MovementMode}
   Save cached pose 'PreAimOffsetsInputPose'(Pose=Input Pose!() -> (Pose) {name=InSourcePosePreAimOffsets}) {cacheName=PreAimOffsetsInputPose}
   Save cached pose 'AimOffsetInputPose'(Pose=Input Pose() {name=InAimOffsetsPose}) {cacheName=AimOffsetInputPose}
  ]}
 Single Reload Aim SM {
  entry="PassThrough (Reload Both)"
  path=MovementMode_AimOffset.AnimGraphNode_StateMachine_0.Single Reload Aim SM
  ids=id
  nodes [
   AnimStateNode_3 = Both Reload Single() {is=title AnimStateNode}
   AnimStateNode_4 = "PassThrough (Reload Both)"() {is=title AnimStateNode}
   AnimStateNode_5 = Right Reload Single() {is=title AnimStateNode}
   AnimStateNode_6 = Left Reload Single() {is=title AnimStateNode}
   AnimStateTransitionNode_10 = AnimStateTransitionNode()
   AnimStateTransitionNode_11 = AnimStateTransitionNode()
   AnimStateTransitionNode_12 = AnimStateTransitionNode()
   AnimStateTransitionNode_13 = AnimStateTransitionNode() {crossfadeDuration=0.3}
   AnimStateTransitionNode_14 = AnimStateTransitionNode()
   AnimStateTransitionNode_4 = AnimStateTransitionNode()
   AnimStateTransitionNode_5 = AnimStateTransitionNode()
   AnimStateTransitionNode_6 = AnimStateTransitionNode()
   AnimStateTransitionNode_7 = AnimStateTransitionNode()
   AnimStateTransitionNode_8 = AnimStateTransitionNode() {crossfadeDuration=0.3}
   AnimStateTransitionNode_9 = AnimStateTransitionNode()
  ]
  links [
   {
    dst=Left Reload Single
    rule=MovementMode_AimOffset.AnimGraphNode_StateMachine_0.Single Reload Aim SM.AnimStateTransitionNode_5.Transition
    src=Both Reload Single
    via=AnimStateTransitionNode_5
   }
   {
    dst="PassThrough (Reload Both)"
    rule=MovementMode_AimOffset.AnimGraphNode_StateMachine_0.Single Reload Aim SM.AnimStateTransitionNode_7.Transition
    src=Both Reload Single
    via=AnimStateTransitionNode_7
   }
   {
    dst=Right Reload Single
    rule=MovementMode_AimOffset.AnimGraphNode_StateMachine_0.Single Reload Aim SM.AnimStateTransitionNode_6.Transition
    src=Both Reload Single
    via=AnimStateTransitionNode_6
   }
   {
    dst=Both Reload Single
    rule=MovementMode_AimOffset.AnimGraphNode_StateMachine_0.Single Reload Aim SM.AnimStateTransitionNode_12.Transition
    src=Left Reload Single
    via=AnimStateTransitionNode_12
   }
   {
    dst=Both Reload Single
    rule=MovementMode_AimOffset.AnimGraphNode_StateMachine_0.Single Reload Aim SM.AnimStateTransitionNode_4.Transition
    src=Left Reload Single
    via=AnimStateTransitionNode_4
   }
   {
    dst="PassThrough (Reload Both)"
    rule=MovementMode_AimOffset.AnimGraphNode_StateMachine_0.Single Reload Aim SM.AnimStateTransitionNode_13.Transition
    src=Left Reload Single
    via=AnimStateTransitionNode_13
   }
   {
    dst=Left Reload Single
    rule=MovementMode_AimOffset.AnimGraphNode_StateMachine_0.Single Reload Aim SM.AnimStateTransitionNode_9.Transition
    src="PassThrough (Reload Both)"
    via=AnimStateTransitionNode_9
   }
   {
    dst=Right Reload Single
    rule=MovementMode_AimOffset.AnimGraphNode_StateMachine_0.Single Reload Aim SM.AnimStateTransitionNode_10.Transition
    src="PassThrough (Reload Both)"
    via=AnimStateTransitionNode_10
   }
   {
    dst=Both Reload Single
    rule=MovementMode_AimOffset.AnimGraphNode_StateMachine_0.Single Reload Aim SM.AnimStateTransitionNode_11.Transition
    src=Right Reload Single
    via=AnimStateTransitionNode_11
   }
   {
    dst=Both Reload Single
    rule=MovementMode_AimOffset.AnimGraphNode_StateMachine_0.Single Reload Aim SM.AnimStateTransitionNode_14.Transition
    src=Right Reload Single
    via=AnimStateTransitionNode_14
   }
   {
    dst="PassThrough (Reload Both)"
    rule=MovementMode_AimOffset.AnimGraphNode_StateMachine_0.Single Reload Aim SM.AnimStateTransitionNode_8.Transition
    src=Right Reload Single
    via=AnimStateTransitionNode_8
   }
  ]
 }
 Both Reload Single {
  path=MovementMode_AimOffset.AnimGraphNode_StateMachine_0.Single Reload Aim SM.AnimStateNode_3.Both Reload Single
  nodes [
   n1 = Layered blend per bone(BasePose=Use cached pose 'PreAimOffsetsInputPose'(), BlendPoses_0=Use cached pose 'AimOffsetInputPose'()) {
    bMeshSpaceRotationBlend=true
    layerSetup [{branchFilters [{blendDepth=-1, boneName=clavicle_l}, {boneName=root}, {blendDepth=-1, boneName=clavicle_r}]}]
   }
   Output Animation Pose(Result=@n1)
  ]
 }
 "PassThrough (Reload Both)" {
  path="MovementMode_AimOffset.AnimGraphNode_StateMachine_0.Single Reload Aim SM.AnimStateNode_4.PassThrough (Reload Both)"
  nodes [Output Animation Pose(Result=Use cached pose 'AimOffsetInputPose'())]
 }
 Right Reload Single {
  path=MovementMode_AimOffset.AnimGraphNode_StateMachine_0.Single Reload Aim SM.AnimStateNode_5.Right Reload Single
  nodes [
   n1 = Layered blend per bone(BasePose=Use cached pose 'PreAimOffsetsInputPose'(), BlendPoses_0=Use cached pose 'AimOffsetInputPose'()) {bMeshSpaceRotationBlend=true, layerSetup [{branchFilters [{boneName=root}, {blendDepth=-1, boneName=clavicle_r}]}]}
   Output Animation Pose(Result=@n1)
  ]
 }
 Left Reload Single {
  path=MovementMode_AimOffset.AnimGraphNode_StateMachine_0.Single Reload Aim SM.AnimStateNode_6.Left Reload Single
  nodes [
   n1 = Layered blend per bone(BasePose=Use cached pose 'PreAimOffsetsInputPose'(), BlendPoses_0=Use cached pose 'AimOffsetInputPose'()) {bMeshSpaceRotationBlend=true, layerSetup [{branchFilters [{boneName=root}, {blendDepth=-1, boneName=clavicle_l}]}]}
   Output Animation Pose(Result=@n1)
  ]
 }
 Transition {
  path=MovementMode_AimOffset.AnimGraphNode_StateMachine_0.Single Reload Aim SM.AnimStateTransitionNode_10.Transition
  nodes [Result(bCanEnterTransition=Get bIsPrimaryReloadStart())]
 }
 Transition {
  path=MovementMode_AimOffset.AnimGraphNode_StateMachine_0.Single Reload Aim SM.AnimStateTransitionNode_11.Transition
  nodes [Result(bCanEnterTransition=Get bIsReloadBoth() -> (bIsReloadBoth:bool))]
 }
 Transition {
  path=MovementMode_AimOffset.AnimGraphNode_StateMachine_0.Single Reload Aim SM.AnimStateTransitionNode_12.Transition
  nodes [Result(bCanEnterTransition=Get bIsReloadBoth())]
 }
 Transition {
  path=MovementMode_AimOffset.AnimGraphNode_StateMachine_0.Single Reload Aim SM.AnimStateTransitionNode_13.Transition
  nodes [Result(bCanEnterTransition=Get bIsEarlySecondaryReloadBlendOut())]
 }
 Transition {
  path=MovementMode_AimOffset.AnimGraphNode_StateMachine_0.Single Reload Aim SM.AnimStateTransitionNode_14.Transition
  nodes [Result(bCanEnterTransition=Get bIsReloadBothSingle() -> (bIsReloadBothSingle:bool))]
 }
 Transition {
  path=MovementMode_AimOffset.AnimGraphNode_StateMachine_0.Single Reload Aim SM.AnimStateTransitionNode_4.Transition
  nodes [Result(bCanEnterTransition=Get bIsReloadBothSingle())]
 }
 Transition {
  path=MovementMode_AimOffset.AnimGraphNode_StateMachine_0.Single Reload Aim SM.AnimStateTransitionNode_5.Transition
  nodes [Result(bCanEnterTransition=Get bIsSecondaryReloadOnly() -> (bIsSecondaryReloadOnly:bool))]
 }
 Transition {
  path=MovementMode_AimOffset.AnimGraphNode_StateMachine_0.Single Reload Aim SM.AnimStateTransitionNode_6.Transition
  nodes [Result(bCanEnterTransition=Get bIsPrimaryReloadOnly() -> (bIsPrimaryReloadOnly:bool))]
 }
 Transition {
  path=MovementMode_AimOffset.AnimGraphNode_StateMachine_0.Single Reload Aim SM.AnimStateTransitionNode_7.Transition
  nodes [Result(bCanEnterTransition=NOT Boolean(A=Get bIsReloadSingle() -> (bIsReloadSingle:bool)))]
 }
 Transition {
  path=MovementMode_AimOffset.AnimGraphNode_StateMachine_0.Single Reload Aim SM.AnimStateTransitionNode_8.Transition
  nodes [Result(bCanEnterTransition=Get bIsEarlyPrimaryReloadBlendOut())]
 }
 Transition {
  path=MovementMode_AimOffset.AnimGraphNode_StateMachine_0.Single Reload Aim SM.AnimStateTransitionNode_9.Transition
  nodes [Result(bCanEnterTransition=Get bIsSecondaryReloadStart())]
 }
 MovementMode_BaseLocomotion {nodes [
   Input Pose() {name=InPoseBaseLocomotion}
   Output Pose() {layerGroup=MovementMode}
  ]}
 MovementMode_FinalPoseOverride {nodes [
   Input Pose() {name=InFinalPose}
   Output Pose() {layerGroup=MovementMode}
  ]}
 MovementMode_FullBodyOverride {nodes [
   Input Pose() {name=InputPoseFullBody}
   Output Pose() {layerGroup=MovementMode}
  ]}
 MovementMode_InAirOverride {nodes [
   Input Pose() {name=InSource_Pose}
   Input Pose() {name=InAirPose}
   Output Pose() {layerGroup=MovementMode}
  ]}
 MovementMode_LocomotionAdditives {nodes [
   Input Pose() {name=InLocomotionAdditivesPose}
   Output Pose() {layerGroup=MovementMode}
  ]}
 MovementMode_LowerBodyOverride {nodes [
   Input Pose() {name=InLowerBodyPose}
   Input Pose() {name=InSourcePose}
   Output Pose() {layerGroup=MovementMode}
  ]}
 MovementMode_OffScreen {nodes [
   Input Pose() {name=InPoseOffscreen}
   Output Pose() {layerGroup=MovementMode}
  ]}
 MovementMode_SourcePose {nodes [
   Input Pose() {name=InSourcePoseBase}
   Output Pose() {layerGroup=MovementMode}
  ]}
]
variables []
